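; PCG Blueprint element that turns a static mesh's sockets into PCG points.
; It gets the StaticMesh variable's sockets matching Tag and makes a point from each relative transform, calling SetSeedfromPosition and collecting the points.
; It then builds a PCGPointData from those points, targets it at the component's owner and returns it as a data collection.
; The Input collection is unused, and no actor transform is applied to the socket transforms.
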(fn ExecuteWithContext
  (params (InContext :type PCGContext%20Structure%20%28by%20ref%29) (Input :type PCGData%20Collection%20Structure%20%28by%20ref%29))
  (Utilities|FlowControl|Sequence
    (:then_0
      ;[1 1 1] Loop over the static mesh actor's sockets, match by tag and create PCG Points.
      (for _array_element (StaticMesh|GetSocketsbyTag (Variables|Default|GetStaticMesh) (Variables|Default|GetTag))
        (PCG|Helpers|SetSeedfromPosition (Utilities|Struct|MakePCGPoint (Math|Transform|MakeTransform (Variables|/Script/Engine.StaticMeshSocket|GetRelativeLocation _array_element) (Variables|/Script/Engine.StaticMeshSocket|GetRelativeRotation _array_element) (Variables|/Script/Engine.StaticMeshSocket|GetRelativeScale _array_element))))
        (Utilities|Array|Add (LocalVariables|GetOutPoints) (Utilities|Struct|MakePCGPoint (Math|Transform|MakeTransform (Variables|/Script/Engine.StaticMeshSocket|GetRelativeLocation _array_element) (Variables|/Script/Engine.StaticMeshSocket|GetRelativeRotation _array_element) (Variables|/Script/Engine.StaticMeshSocket|GetRelativeScale _array_element))))))
    (:then_1
      ;[1 1 1] Build the PCG Point Data, Set its Points and Target Actor
      (bind _returnvalue (PCG|Temporary|GetComponent InContext))
      (bind _returnvalue_2 (Game|ConstructObjectfromClass "/Script/PCG.PCGPointData"))
      (Variables|/Script/PCG.PCGSpatialData|SetTargetActor (Components|GetOwner _returnvalue) _returnvalue_2)
      (SpatialData|SetPoints _returnvalue_2 (LocalVariables|GetOutPoints))
      (return (Utilities|Struct|MakePCGDataCollection (Utilities|Array|MakeArray (PCG|TaggedData|MakeTaggedData _returnvalue_2)))))))
